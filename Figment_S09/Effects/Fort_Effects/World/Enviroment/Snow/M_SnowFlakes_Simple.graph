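# Unlit, translucent, two-sided particle sprite material for snowflakes, tinted by particle color and a time-of-day color that shifts to blue at night.
# A random RGB channel pick per particle shapes opacity from snow textures, blending sharp-times-blurred with blurred-only by pixel depth and particle speed.
# Near the camera, flakes shrink toward particle center and fade, both bypassed on mobile feature levels.
# A static switch adds optional distance-cull fading.
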Material /Figment_S09/Effects/Fort_Effects/World/Enviroment/Snow/M_SnowFlakes_Simple
graph {
 properties {
  bEnableResponsiveAA=true
  bUsedWithParticleSprites=true
  blendMode=BLEND_Translucent
  pixelDepthOffsetMode=PDOM_Legacy
  shadingModel=MSM_Unlit
  translucencyLightingMode=TLM_VolumetricPerVertexDirectional
  twoSided=true
  usedShadingModels=Unlit
 }
 nodes [
  n1 = ScalarParameter() {defaultValue=300, parameterName=ScaleDown_Start}
  n2 = Clamp(None=Subtract(
    A=MaterialFunctionCall(Distance=@n1) {materialFunction=&Near_Camera_Fade}.Result
    B=ScalarParameter() {parameterName=ScaleDown_End_Normalized}
   ))
  n3 = LinearInterpolate(A=Subtract(A=ParticlePositionWS(), B=WorldPosition().XYZ), B=Constant(), Alpha=@n2)
  n4 = Constant()
  FeatureLevelSwitch(Default=@n3, ES2=@n4, ES3_1=@n4, SM4_REMOVED=@n3)
  n5 = VectorParameter() {parameterName=MorningColor, defaultValue=<rgba 0.5 0.3976 0.3042 0>}
  n6 = ParticleColor()
  Multiply(
   A=MaterialFunctionCall(
    Color_Morning=@n5.RGB
    Color_Evening=@n5.RGB
    Color_Night=VectorParameter() {parameterName=NightColor, defaultValue=<rgba 0.1906 0.268 0.5 0>}.RGB
   ) {materialFunction=&Color_Time_of_Day}.Result
   B=@n6.RGB
  )
  n7 = ParticleRandom()
  n8 = TextureSample() {texture=/Game/Weapons/WeaponSkins/Elemental/Textures/T_Snow_01_blur}
  n9 = PixelDepth()
  n10 = Constant()
  n11 = ScalarParameter() {defaultValue=800, parameterName=NoBlurRange_Centered}
  n12 = SphereMask(A=PixelDepth().R, B=Constant(), Radius=@n11) {hardnessPercent=10}
  n13 = Abs(None=Subtract(A=SphereMask(A=@n9.R, B=@n10, Radius=@n11) {hardnessPercent=10}, B=OneMinus(None=@n12)))
  n14 = Divide(A=ParticleSpeed(), B=ScalarParameter() {defaultValue=300, parameterName=SpeedBlurThreshold}) {constB=400}
  n15 = LinearInterpolate(
   A=Multiply(A=TextureSample() {texture=/Game/Weapons/WeaponSkins/Elemental/Textures/T_Snow_01}.RGB, B=@n8.RGB)
   B=@n8.RGB
   Alpha=Clamp(None=Subtract(A=@n13, B=@n14))
  )
  n16 = Constant() {r=1}
  n17 = Multiply(
   A=Multiply(
    A=@n6.A
    B=MaterialFunctionCall!(Random Picker=@n7, Texture Input=@n15, Option 1, Option 2, Option 3) {materialFunction=/Game/Effects/Fort_Effects/Materials/_Functions/MF_RGB_Random_01}.Result
   )
   B=FeatureLevelSwitch(Default=@n2, ES2=@n16, ES3_1=@n16, SM4_REMOVED=@n2)
  )
  StaticSwitchParameter(True=Multiply(A=DistanceCullFade(), B=@n17), False=@n17) {parameterName=UseDistanceCullFade}
 ]
}
